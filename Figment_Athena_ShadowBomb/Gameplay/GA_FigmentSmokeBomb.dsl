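; Gameplay ability for the Figment Shadow Bomb that gives and removes the smoke bomb gadget item in a player's inventory.
; GiveItem adds AGID_SmokeBomb to the controller's inventory, which per its node note also changes character parts.
; RemoveItem force-removes that item with an amount of -1.
; The activate and end-ability events are disabled placeholders, and this graph does not call either function itself.
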(event Ability|EventActivateAbility
  (|CallParentFunction) ; This node is disabled and will not be called.\nDrag off pins to build functionality.
)

(event Ability|EventOnEndAbility
  (params (bWasCancelled :type Boolean))
  (|CallParentFunction bWasCancelled) ; This node is disabled and will not be called.\nDrag off pins to build functionality.
)

(fn GiveItem
  (params (Controller :type Fort%20Player%20Controller%20Object%20Reference))
  (Inventory|GiveItemtoInventoryOwner Controller "/Game/Athena/Items/Gameplay/SmokeBomb/AGID_SmokeBomb.AGID_SmokeBomb") ; This is the node that gives the gadget and changes character parts
)

(fn RemoveItem
  (params (Controller :type Fort%20Player%20Controller%20Object%20Reference))
  (Inventory|RemoveItemFromPlayer :PlayerController Controller :ItemDefinition "/Game/Athena/Items/Gameplay/SmokeBomb/AGID_SmokeBomb.AGID_SmokeBomb" :AmountToRemove -1 :bForceRemoval true :bSkipVerbRemoval false))
