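; Gameplay ability for the Figment Shadow Bomb: plays the use montage, spawns a smoke bomb projectile, and applies the shroud effect to the owner.
; On the montage trigger, a short timer delays the shroud to sync with the smoke, then commits and ends the ability.
; A failsafe timer also ends the ability.
; An inventory overflow drop removes the shadow bomb effects and the associated item from the player, then ends the ability.
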(event Custom|FailsafeEnd
  (Ability|EndAbility))

(event Ability|EventActivateAbility
  (|CallParentFunction)
  (EventDispatchers|BindEventtoOnItemDroppedDuetoOverflow (Utilities|Casting|CastToFortPlayerControllerAthena (Fortnite|GetFortPlayerControllerfromActor (Ability|GetAvatarActorfromActorInfo))) (delegate Custom|CleanupShadowbomb))
  (Ability|Tasks|PlayMontageandWaitforTrigger "None" (Variables|Default|GetSmokeBombMontage) 1.1
    (:then
      (Utilities|Time|SetTimerbyEvent (delegate Custom|FailsafeEnd) 3.0) ; Failsafe Timer
    )
    (:Triggered
      (Projectile|SpawnProjectile "/Figment_Athena_ShadowBomb/B_Figment_Prj_Athena_SmokeBomb.B_Figment_Prj_Athena_SmokeBomb_C" (Ability|GetAvatarActorfromActorInfo) (- (Transformation|GetActorLocation (Ability|GetAvatarActorfromActorInfo)) "0, 0,28.000000") (Math|Conversions|RotationFromXVector (Transformation|GetActorLocation (Ability|GetAvatarActorfromActorInfo))) (Math|Conversions|RotationFromXVector (Transformation|GetActorLocation (Ability|GetAvatarActorfromActorInfo)))) ; Spawn the smoke bomb
      (Utilities|Time|SetTimerbyEvent (delegate Custom|ApplyShroud) 0.125) ; Slight delay before applying effect so it syncs with the smoke cloud spawning
    )
    (:Cancelled
      (Utilities|FlowControl|Delay 0.5)
      (Ability|EndAbility))))

(event Custom|ApplyShroud
  (Ability|ApplyGameplayEffectToOwner (Variables|Default|GetGE%20ApplySmokeBomb) 0) ; This is the shroud ability
  (Ability|CommitAbility)
  (Utilities|FlowControl|Delay 0.5)
  (Ability|EndAbility))

(event Custom|CleanupShadowbomb
  (params (ItemDropped :type Fort%20World%20Item%20Object%20Reference))
  (bind _returnvalue (Ability|GetAvatarActorfromActorInfo))
  (bind _returnvalue_2 (Ability|GetAbilitySystemComponent _returnvalue))
  ;[1 1 1] Remove eveyrthing shadowbomb related if the item is dropped
  (GameplayEffects|RemoveActiveGameplayEffectbySourceEffect _returnvalue_2 "/Figment_Athena_ShadowBomb/Gameplay/GE_Figment_ApplySmokeBomb.GE_Figment_ApplySmokeBomb_C" 0 -1)
  (GameplayEffects|RemoveActiveGameplayEffectbySourceEffect _returnvalue_2 "/Game/Athena/Items/Consumables/SmokeBomb_Athena/GE_ShadowBOmbLoop.GE_ShadowBombLoop_C" 0 -1)
  (Inventory|RemoveItemFromPlayer :PlayerController (Fortnite|GetFortPlayerControllerfromActor _returnvalue) :ItemDefinition "/Game/Athena/Items/Gameplay/SmokeBomb/AGID_SmokeBomb.AGID_SmokeBomb" :AmountToRemove -1 :bForceRemoval true :bSkipVerbRemoval false)
  (Ability|EndAbility))
